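# UI-domain material that draws the blue background of an SW series Item Shop tile.
# A rotated, smoothstepped linear gradient blends navy to mid blue, then is faintly darkened by the inverted, center-scaled SW pattern texture.
# An offset exponential radial gradient blends in a brighter blue glow over that result.
# A UVFit named reroute supplies fill-scaled UVs; separate Add, Multiply and screen-space UV nodes feed nothing visible in the graph.
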Material /ItemShop/UI/Art/SW2024/M_UI_SWSeries_ShopTile
graph {
 properties {bUsedWithStaticMesh=false, materialDomain=MD_UI, pixelDepthOffsetMode=PDOM_Legacy, usedShadingModels=Default Lit}
 nodes [
  n1 = Constant3Vector() {constant=<rgba 0 0.3667 0.9774 1>}
  n2 = NamedRerouteUsage()
  n3 = Constant() {r=1}
  n4 = Constant() {r=-0.5}
  n5 = Constant() {r=0.25}
  n6 = Multiply(A=MaterialFunctionCall(
    UVs=MaterialFunctionCall!(UVs=@n2, Scale=@n3, X Position=@n4, Y Position=@n5) {materialFunction=/OfferCatalog/Art/_MaterialFunctions_ShopTheming/MF_UI_ScalePosition}.Result
    Radius=Constant() {r=0.75}
    Density=Constant() {r=1.2}
   ) {materialFunction=&RadialGradientExponential}.RadialGradientExponential) {constB=0.3}
  Add(A=@n1, B=@n6)
  n7 = Constant3Vector() {constant=<rgba 0.0042 0.0261 0.1719 1>}
  n8 = Constant3Vector() {constant=<rgba 0.013 0.0953 0.6038 1>}
  n9 = SmoothStep(Value=MaterialFunctionCall(UVs=NamedRerouteUsage(), GradientStart=Constant(), GradientEnd=Constant() {r=1}, RotationAngle=Constant() {r=0.28}) {materialFunction=&MF_UI_LinearGradient}.Result)
  n10 = NamedRerouteUsage()
  n11 = Constant() {r=1.7}
  n12 = OneMinus(None=TextureSample(UVs=MaterialFunctionCall(UVs=@n10, Texture Scale=@n11) {materialFunction=&ScaleUVsByCenter}.UVs) {samplerType=SAMPLERTYPE_Alpha, texture=/Game/UI/Frontend/ItemShop/Art/SW2024/T_UI_SWPattern}.R)
  LinearInterpolate(A=LinearInterpolate(A=LinearInterpolate(A=@n7, B=@n8, Alpha=@n9), Alpha=Multiply(A=@n12) {constB=0.12}) {constB=0}, B=@n1, Alpha=@n6)
  MaterialFunctionCall!(ScreenUV, TextureSize, ScalePivotPoint) {materialFunction=/FortUILibrary/MaterialFunctions/Basics/MF_UI_ScreenSpaceUVScaling}
  Multiply(A=@n6, B=@n1)
  NamedRerouteDeclaration(None=MaterialFunctionCall() {materialFunction=&MF_UI_UVScaling}.Scaled To Fill) {name=UVFit, nodeColor=<rgba 1 0.8672 0 1>}
 ]
}
